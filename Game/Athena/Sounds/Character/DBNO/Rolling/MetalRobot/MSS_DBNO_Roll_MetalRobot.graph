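# MetaSound preset of MSS_DBNO_Roll that plays the mono one-shot foley for metal robot characters rolling while down but not out.
# It supplies generic dirt roll waves as main sounds and MetalRobot running clothing foley as sweeteners, with overrides for pitch, volume and sweetener level.
# Playback logic lives in the parent; this asset also assigns third-person sprint footstep attenuation, the player foley sound class and a DBNO roll audio properties sheet.
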MetaSoundSource /Game/Athena/Sounds/Character/DBNO/Rolling/MetalRobot/MSS_DBNO_Roll_MetalRobot
preset_parent=/Game/Athena/Sounds/Character/DBNO/Rolling/Settings/MSS_DBNO_Roll
inputs [
 Clothing Sweeteners {type=WaveAsset:Array}
 Random.Pitch.Base {type=Float}
 Random.Pitch.Spread {type=Float}
 Random.Volume.Base {type=Float}
 Random.Volume.Cloth Sweetener Volume {type=Float}
 Random.Volume.Spread {type=Float}
 Sounds {type=WaveAsset:Array}
 UE.Source.OnPlay {type=Trigger}
]
interfaces [UE.OutputFormat.Mono:1.0, UE.Source.OneShot:1.0, UE.Source:1.1]
outputs [
 "UE.OutputFormat.Mono.Audio:0" {type=Audio}
 UE.Source.OneShot.OnFinished {type=Trigger}
]
pages [Default {input_defaults {
   Random.Pitch.Base=0
   Random.Pitch.Spread=0.5
   Random.Volume.Base=6
   Random.Volume.Cloth Sweetener Volume=-3
   Random.Volume.Spread=0
   UE.Source.OnPlay=null
   Clothing Sweeteners [
    /Game/Sounds/Fort_Player/Clothing/MetalRobot/Waves/Foley_Clothing_MetalRobot_Run_01
    /Game/Sounds/Fort_Player/Clothing/MetalRobot/Waves/Foley_Clothing_MetalRobot_Run_02
    /Game/Sounds/Fort_Player/Clothing/MetalRobot/Waves/Foley_Clothing_MetalRobot_Run_03
    /Game/Sounds/Fort_Player/Clothing/MetalRobot/Waves/Foley_Clothing_MetalRobot_Run_04
    /Game/Sounds/Fort_Player/Clothing/MetalRobot/Waves/Foley_Clothing_MetalRobot_Run_05
    /Game/Sounds/Fort_Player/Clothing/MetalRobot/Waves/Foley_Clothing_MetalRobot_Run_06
   ]
   Sounds [
    /Game/Athena/Sounds/Character/DBNO/Rolling/Generic/Waves/Dirt/DBNO_Roll_Dirt_01
    /Game/Athena/Sounds/Character/DBNO/Rolling/Generic/Waves/Dirt/DBNO_Roll_Dirt_02
    /Game/Athena/Sounds/Character/DBNO/Rolling/Generic/Waves/Dirt/DBNO_Roll_Dirt_03
    /Game/Athena/Sounds/Character/DBNO/Rolling/Generic/Waves/Dirt/DBNO_Roll_Dirt_04
    /Game/Athena/Sounds/Character/DBNO/Rolling/Generic/Waves/Dirt/DBNO_Roll_Dirt_05
   ]
  }}]
properties {
 assetImportData=&self:AssetImportData
 attenuationSettings=/Game/Sounds/Attenuation/Fort_Player_Footsteps_3P_Sprint_Athena_Att
 audioPropertiesSheet=/Game/Athena/Sounds/Character/DBNO/Rolling/Settings/APS_Foley_DBNORoll
 maxDistance=3584
 sampleRate=48000
 soundClassObject=&Fort_Player_Foley_Athena
 volume=1.1
}
